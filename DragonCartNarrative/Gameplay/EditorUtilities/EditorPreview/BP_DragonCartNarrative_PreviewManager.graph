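# Editor utility that applies and unloads Dragon Cart Narrative preview profiles in the level editor.
# LoadProfile chains data layer filtering by name, World Partition loading through temporary, timer-destroyed LocationVolumes, viewport camera placement, time-of-day and DSMV volume loading.
# UnloadProfile unloads the World Partition region the same way.
# Time-of-day stages are empty; helpers for manager, cubemap parameters and a Python-set world setting are not called here.
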EditorUtilityBlueprint /DragonCartNarrative/Gameplay/EditorUtilities/EditorPreview/BP_DragonCartNarrative_PreviewManager
graphs [
 ActorsToGuid {nodes [
   n1 = ActorsToGuid!() -> (then, Actors:Array of Actor Object References) {is=title K2Node_FunctionEntry}
   > n2 = For Each Loop(Array:Array of Actor Object References=@n1) -> (Array Element:Actor Object Reference)
   Return Node!(execute=@n2.Completed, Guids:Array of Guid Structures=Get GuidsArray() -> (GuidsArray:Array of Guid Structures))
   n3 = Get ActorGuid!(self:Actor Object Reference=@n2) -> (ActorGuid:Guid Structure)
   > Add(TargetArray:Array of Guid Structures=Get GuidsArray(), NewItem:"Guid Structure (by ref)"=@n3) {is=title K2Node_CallArrayFunction}
   > PrintString(WorldContextObject=@n2, InString="To String (Guid)"!(InGuid:"Guid Structure (by ref)"=@n3) -> (ReturnValue:String))
  ]}
 DoesLayersContainName {nodes [
   n1 = DoesLayersContainName!() -> (then, Layers:Array of Data Layer Asset Object References, Name:String) {is=title K2Node_FunctionEntry}
   > n2 = For Each Loop(Array:Array of Data Layer Asset Object References=@n1) -> (Array Element:Data Layer Asset Object Reference)
   Return Node!(execute=@n2.Completed, ContainName:bool=false)
   > Branch(Condition="Equal, Case Insensitive (String)"(A=GetObjectName(Object=@n2), B=@n1.Name))
   > Return Node(ContainName=true)
  ]}
 EventGraph {nodes [
   n1 = LoadProfile!() -> (OutputDelegate, then, EditorPreviewProfile:PDA Dragon Cart Narrative Editor Preview Profile Data Object Reference) {is=title K2Node_CustomEvent}
   > K2Node_Composite_3 = DataLayers_On!(execute, Profile:PDA Dragon Cart Narrative Editor Preview Profile Data Object Reference=@n1.EditorPreviewProfile, Active:bool=true) -> (then:Exec) {is=title K2Node_Composite}
   > K2Node_Composite_10 = World Partition_On!(execute, Profile:PDA Dragon Cart Narrative Editor Preview Profile Data Object Reference=@n1.EditorPreviewProfile, Load:bool=true) -> (then:Exec, else:Exec) {is=title K2Node_Composite}
   K2Node_Composite_2 = Camera!(
    execute=[@K2Node_Composite_10.else, @K2Node_Composite_10.then]
    Profile:PDA Dragon Cart Narrative Editor Preview Profile Data Object Reference=@n1.EditorPreviewProfile
   ) -> (then:Exec) {is=title K2Node_Composite}
   > K2Node_Composite_12 = TODM_On!(execute, Profile:PDA Dragon Cart Narrative Editor Preview Profile Data Object Reference=@n1.EditorPreviewProfile, Active:bool=true) -> (then:Exec) {is=title K2Node_Composite}
   > K2Node_Composite_4 = DSMV!(execute, Config:PDA Dragon Cart Narrative Editor Preview Profile Data Object Reference=@n1.EditorPreviewProfile) -> (then:Exec) {is=title K2Node_Composite}
   n2 = UnloadProfile!() -> (OutputDelegate, then, EditorPreviewProfile:PDA Dragon Cart Narrative Editor Preview Profile Data Object Reference) {is=title K2Node_CustomEvent}
   > K2Node_Composite_7 = World Partition_Off!(execute, Profile:PDA Dragon Cart Narrative Editor Preview Profile Data Object Reference=@n2.EditorPreviewProfile, Load:bool) -> (then:Exec, else:Exec) {is=title K2Node_Composite}
   K2Node_Composite_8 = TODM_Off!(
    execute=[@K2Node_Composite_7.else, @K2Node_Composite_7.then]
    Profile:PDA Dragon Cart Narrative Editor Preview Profile Data Object Reference=@n2.EditorPreviewProfile
    Active:bool
   ) -> (then:Exec) {is=title K2Node_Composite}
  ]}
 World Partition_On {
  path=EventGraph.K2Node_Composite_10.World Partition_On
  nodes [
   n1 = CleanupVolumes() {is=title K2Node_CustomEvent}
   n2 = Inputs!() -> (execute, Profile:PDA Dragon Cart Narrative Editor Preview Profile Data Object Reference, Load:bool=~)
   n3 = Get!(self:PDA Dragon Cart Narrative Editor Preview Profile Data Object Reference=@n2) -> (
    WorldPartitionData_RequiresLoading_6_#1:bool=~
    WorldPartitionData_Location_2_#2:Vector=~
    WorldPartitionData_BrushSize_4_#3:Vector=~
   )
   n4 = SpawnActorFromClass!(
    execute=@n9.then
    ActorClass:Actor Class Reference=/Script/Engine.LocationVolume
    Location:Vector=@n3.WorldPartitionData_Location_2_#2
    Rotation:Rotator
    bTransient:bool
   ) -> (then, ReturnValue:Location Volume Object Reference)
   n5 = Set Timer by Event(execute=[@n8.then, @n7.then], Delegate=@n1, Time=0.25, bLooping=false)
   Clear and Invalidate Timer by Handle!(execute=@n1.then, WorldContextObject:Object Reference=@n4, Handle=@n5) -> (then)
   > Destroy Actor(self=@n4)
   Outputs!(then:Exec=@n5.then, else:Exec=@n9.else)
   SetActorScale3D!(execute=@n4.then, self:Actor Object Reference=@n4, NewScale3D:Vector=@n3.WorldPartitionData_BrushSize_4_#3) -> (then)
   > n6 = Branch(Condition=@n2.Load)
   > n7 = Load(self:Location Volume Object Reference=@n4)
   n8 = Unload(execute=@n6.else, self:Location Volume Object Reference=@n4)
   n9 = Branch(execute=@n2.execute, Condition=@n3)
  ]
 }
 TODM_On {
  path=EventGraph.K2Node_Composite_12.TODM_On
  nodes [
   Inputs!() -> (execute, Profile, Active:bool=~)
   Outputs!(then)
  ]
 }
 Camera {
  path=EventGraph.K2Node_Composite_2.Camera
  nodes [
   n1 = Inputs!() -> (execute, Profile)
   n2 = Get!(self=@n1) -> (
    SillyTunaProfileData_SetEditorCamera_1_#4:bool=~
    SillyTunaProfileData_Location_6_#5:Vector=~
    SillyTunaProfileData_Rotation_7_#6:Rotator=~
    SillyTunaProfileData_RiftScale_10_#7:double=~
   )
   > n3 = Branch(Condition=@n2)
   > n4 = SetLevelViewportCameraInfo!(
    execute
    self:Unreal Editor Subsystem Object Reference=Get UnrealEditorSubsystem()
    CameraLocation:Vector=@n2.SillyTunaProfileData_Location_6_#5
    CameraRotation:Rotator=@n2.SillyTunaProfileData_Rotation_7_#6
   ) -> (then)
   Outputs(then=[@n4.then, @n3.else])
  ]
 }
 DataLayers_On {
  path=EventGraph.K2Node_Composite_3.DataLayers_On
  nodes [
   n1 = Inputs!() -> (execute, Profile, Active:bool=~)
   > ToggleVisibilityForLayers!(
    execute
    self
    Layers:Array of Data Layer Asset Object References=Get!(self=@n1) -> (DataLayerData_DataLayersToShow_3_#8:Array of Data Layer Asset Object References)
    Visible:bool=@n1.Active
   ) -> (then)
   > Outputs()
  ]
 }
 DSMV {
  path=EventGraph.K2Node_Composite_4.DSMV
  nodes [
   n1 = Inputs!() -> (execute, Config:PDA Dragon Cart Narrative Editor Preview Profile Data Object Reference)
   > n2 = Branch()
   > n3 = For Each Loop!(
    Exec
    Array:Array of F Dragon Cart Narrative Editor DSMV Data Structures=Get!(self=@n1) -> (
     TimeOfDayData_TimeOfDayManager_27_#9:Info Soft Class Reference
     TimeOfDayData_DSMVLoadLocations_14_#10:Array of F Dragon Cart Narrative Editor DSMV Data Structures
     TimeOfDayData_CubemapIndex_30_#11:int
     TimeOfDayData_CubemapIntensity_33_#12:double=~
    ).TimeOfDayData_DSMVLoadLocations_14_#10
   ) -> (LoopBody, Array Element_LoadLocation_3_#13:Vector=~, Array Index, Completed)
   > n4 = SpawnActorFromClass(ActorClass=&LocationVolume, Location=@n3)
   > Set Box Brush Size!(execute, Volume:Actor Object Reference=@n4, Size_Z:double=1, Size_Y:double=1, Size_X:double=1) -> (then) {is=title K2Node_CallFunction}
   > Load(self=@n4)
   > n5 = SetActorLabel(
    self=@n4
    NewActorLabel="To String (Name)"(InName=Get Name!(self:PDA Dragon Cart Narrative Editor Preview Profile Data Object Reference=@n1) -> (Name:Name=None))
    bMarkDirty=true
   )
   Outputs(then=[@n5.then, @n2.else])
   Get TimeOfDayData!(self:Object Reference=@n1) -> (TimeOfDayData:Structure)
  ]
 }
 World Partition_Off {
  path=EventGraph.K2Node_Composite_7.World Partition_Off
  nodes [
   n1 = CleanupVolume() {is=title K2Node_CustomEvent}
   n2 = Inputs!() -> (execute, Profile:PDA Dragon Cart Narrative Editor Preview Profile Data Object Reference, Load:bool=~)
   n3 = Get!(self=@n2) -> (
    WorldPartitionData_RequiresLoading_6_#1:bool=~
    WorldPartitionData_Location_2_#2:Vector=~
    WorldPartitionData_BrushSize_4_#3:Vector=~
   )
   n4 = SpawnActorFromClass(execute=@n9.then, ActorClass=&LocationVolume, Location=@n3.WorldPartitionData_Location_2_#2)
   n5 = Set Timer by Event(execute=[@n8.then, @n7.then], Delegate=@n1, Time=1, bLooping=false)
   Clear and Invalidate Timer by Handle(execute=@n1.then, WorldContextObject=@n4, Handle=@n5)
   > Destroy Actor(self=@n4)
   Outputs!(then=@n5.then, else:Exec=@n9.else)
   SetActorScale3D(execute=@n4.then, self=@n4, NewScale3D=@n3.WorldPartitionData_BrushSize_4_#3)
   > n6 = Branch(Condition=@n2.Load)
   > n7 = Load(self=@n4)
   n8 = Unload(execute=@n6.else, self=@n4)
   n9 = Branch(execute=@n2.execute, Condition=@n3)
  ]
 }
 TODM_Off {
  path=EventGraph.K2Node_Composite_8.TODM_Off
  nodes [
   Inputs!() -> (execute, Profile, Active:bool=~)
   Outputs!(then)
  ]
 }
 SetPythonParams {nodes [
   n1 = SetPythonParams!() -> (then, Enable:bool=~) {is=title K2Node_FunctionEntry}
   > n2 = Branch(Condition=@n1)
   > n3 = Set FormatAppend(FormatAppend:String=True) -> (Output_Get:String)
   n4 = "To Text (String)"(InString=MakeLiteralString(Value="import unreal\r\n\r\n## Get the current world and its settings\r\nworld = unreal.EditorLevelLibrary.get_editor_world()\r\nworld_settings = world.get_world_settings()\r\n\r\n## Try setting the flag directly using the correct property name\r\ntry:\r\n    world_settings.set_editor_property(\"spawn_time_of_day_manager\", {0})\r\n    world_settings.mark_package_dirty()\r\nexcept Exception as e:\r\n    unreal.log_error(f\"Error setting property 'spawn_time_of_day_manager\")"))
   n5 = Get FormatAppend() -> (FormatAppend:String)
   ExecutePythonCommand(execute=[@n6.then, @n3.then], PythonCommand="To String (Text)"(InText=Format Text!(Format=@n4, 0:String=@n5) -> (Result)))
   > Return Node!(execute)
   n6 = Set FormatAppend(execute=@n2.else, FormatAppend=False)
  ]}
 SetTimeOfDayManagerSettings {nodes [
   SetTimeOfDayManagerSettings!() -> (then, ForceOff:bool=~, TimeOfDayData:F Dragon Cart Narrative Editor Time Of Day Data Structure) {is=title K2Node_FunctionEntry}
   > n1 = GetEditorWorld(self=Get UnrealEditorSubsystem())
   > n2 = Set WorldSettings(WorldSettings:Fort World Settings Object Reference=GetFortWorldSettingsInEditor!(WorldContextObject:Object Reference=@n1) -> (ReturnValue:Fort World Settings Object Reference)) -> (Output_Get:Fort World Settings Object Reference)
   > n3 = Is Valid(InputObject=@n2)
   Return Node(execute=@n3.Is Not Valid)
   > n4 = Branch(Condition=Get ForceOff() -> (ForceOff:bool))
   > n5 = GetTimeOfDayManager!(execute, self:Fort World Settings Object Reference=Get WorldSettings() -> (WorldSettings:Fort World Settings Object Reference)) -> (then, ReturnValue:Fort Time Of Day Manager Interface Interface)
   > SetTimeOfDayManagerVisibility!(execute, self:Fort Time Of Day Manager Interface Interface=@n5, bVisible:bool=false) -> (then)
   > n6 = SetGlobalTimeOfDayManager!(execute, WorldContextObject:Object Reference=Get WorldSettings(), InTimeOfDayManager:Info Soft Class Reference) -> (then)
   Return Node(execute=[@n6.then, @n11.then])
   n7 = Break F Dragon Cart Narrative Editor Time Of Day Data!(F_DragonCartNarrative_Editor_TimeOfDayData:"F Dragon Cart Narrative Editor Time Of Day Data Structure (by ref)"=Get TimeOfDayData!() -> (TimeOfDayData:F Dragon Cart Narrative Editor Time Of Day Data Structure)) -> (
    TimeOfDayManager_27_#9:Info Soft Class Reference
    DSMVLoadLocations_14_#10:Array of F Dragon Cart Narrative Editor DSMV Data Structures
    CubemapIndex_30_#11:int
    CubemapIntensity_33_#12:double=~
   ) {is=title K2Node_BreakStruct}
   Set CubemapIndex(execute=@n4.else, CubemapIndex:int=@n7.CubemapIndex_30_#11) -> (Output_Get:int=0)
   > Set CubemapIntensity(CubemapIntensity:double=@n7.CubemapIntensity_33_#12) -> (Output_Get:double)
   n8 = Break F Dragon Cart Narrative Editor Time Of Day Data(F_DragonCartNarrative_Editor_TimeOfDayData=Get TimeOfDayData()) -> (CubemapIntensity_33_#12=~)
   > SetGlobalTimeOfDayManager(WorldContextObject=Get WorldSettings(), InTimeOfDayManager=@n8)
   > n9 = GetTimeOfDayManager(self=Get WorldSettings())
   > SetTimeOfDayManagerVisibility(self=@n9, bVisible=IsValidSoftClassReference!(SoftClassReference:"Object Soft Class Reference (by ref)"=@n8) -> (ReturnValue:bool))
   n10 = Get WorldSettings()
   > SetScalarParameterValue!(
    execute
    WorldContextObject=@n10
    Collection:Material Parameter Collection Object Reference
    ParameterName:Name=CubemapBlend
    ParameterValue:float="To Float (Integer)"(InInt=Get CubemapIndex() -> (CubemapIndex:int=0))
   ) -> (then) {is=title K2Node_CallMaterialParameterCollectionFunction}
   > n11 = SetScalarParameterValue(WorldContextObject=@n10, ParameterName=CubemapIntensityScale, ParameterValue=Get CubemapIntensity() -> (CubemapIntensity:double))
  ]}
 ToggleVisibilityForLayers {nodes [
   n1 = ToggleVisibilityForLayers!() -> (then, Layers:Array of Data Layer Asset Object References, Visible:bool=~) {is=title K2Node_FunctionEntry}
   n2 = Get DataLayerEditorSubsystem() -> (ReturnValue:Data Layer Editor Subsystem Object Reference) {is=title K2Node_GetEditorSubsystem}
   > n3 = GetAllDataLayers!(execute, self:Data Layer Editor Subsystem Object Reference=@n2) -> (then, ReturnValue:Array of Data Layer Instance Object References)
   > n4 = For Each Loop!(Exec, Array:Array of Data Layer Instance Object References=@n3) -> (LoopBody, Array Element:Data Layer Instance Object Reference, Array Index, Completed)
   > n5 = Branch(Condition=DoesLayersContainName!(
     Layers:Array of Data Layer Asset Object References=@n1
     Name:String=GetDataLayerShortName!(self:Data Layer Instance Object Reference=@n4) -> (ReturnValue:String)
    ) -> (ContainName:bool) {is=title K2Node_CallFunction})
   > n6 = SetDataLayerIsLoadedInEditor!(
    execute
    self:Data Layer Editor Subsystem Object Reference=@n2
    DataLayer:Data Layer Instance Object Reference=@n4
    bIsLoadedInEditor:bool=@n1.Visible
    bIsFromUserChange:bool=true
   ) -> (then, ReturnValue:bool)
   > SetDataLayerVisibility!(execute, self:Data Layer Editor Subsystem Object Reference=@n2, DataLayer:Data Layer Instance Object Reference=@n4, bIsVisible:bool=@n6) -> (then)
   SetDataLayerIsLoadedInEditor(execute=@n5.else, self=@n2, DataLayer=@n4, bIsLoadedInEditor=false, bIsFromUserChange=true)
   > SetDataLayerVisibility(self=@n2, DataLayer=@n4, bIsVisible=false)
  ]}
]
variables [
 GlobalSettings {type=Object, default=null}
 PythonCommand {
  type=string
  default="import unreal## Get the current world and its settingsworld = unreal.EditorLevelLibrary.get_editor_world()world_settings = world.get_world_settings()## Try setting the flag directly using the correct property nametry:    world_settings.set_editor_property(\"spawn_time_of_day_manager\", {0})    world_settings.mark_package_dirty()except Exception as e:    unreal.log_error(f\"Error setting property 'spawn_time_of_day_manager': {str(e)}\")"
 }
]
