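# Button widget representing one selectable option in the DragonCartNarrative editor preview tool: either a preview profile data asset or a reset entry.
# Init stores the IsReset flag, then either labels the button "None" for reset, or stores the Profile and shows its name.
# Clicking the button broadcasts the OnProfileSelected dispatcher with the stored Profile and IsReset flag, leaving the response to listeners.
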WidgetBlueprint /DragonCartNarrative/Gameplay/EditorUtilities/EditorPreview/WBP_DragonCartNarrative_EditorPreviewOption
graphs [
 EventGraph {nodes [
   "On Clicked (Button_0)"() {is=title K2Node_ComponentBoundEvent}
   > Call On Profile Selected!(
    execute:Exec
    self:WBP Dragon Cart Narrative Editor Preview Option Object Reference
    Profile:PDA Dragon Cart Narrative Editor Preview Profile Data Object Reference=Get Profile() -> (Profile:PDA Dragon Cart Narrative Editor Preview Profile Data Object Reference)
    IsReset:bool=Get IsReset() -> (IsReset:bool)
   ) -> (then:Exec) {is=title K2Node_CallDelegate}
   n1 = Init!() -> (OutputDelegate, then, Profile:PDA Dragon Cart Narrative Editor Preview Profile Data Object Reference, IsReset:bool=~) {is=title K2Node_CustomEvent}
   > n2 = Set IsReset(IsReset:bool=@n1.IsReset) -> (Output_Get:bool)
   > n3 = Branch(Condition=@n2)
   > Set Text(Text="LOCTEXT(\"None\")", self=Get Label() -> (Label:Text Object Reference))
   n4 = Set Profile(execute=@n3.else, Profile:PDA Dragon Cart Narrative Editor Preview Profile Data Object Reference=@n1.Profile) -> (Output_Get:PDA Dragon Cart Narrative Editor Preview Profile Data Object Reference)
   > Set Text(
    Text="To Text (Name)"!(InName:Name=Get Name!(self:PDA Dragon Cart Narrative Editor Preview Profile Data Object Reference=@n4) -> (Name:Name=None)) -> (ReturnValue:Text)
    self=Get Label()
   )
  ]}
 OnProfileSelected {nodes [OnProfileSelected!() -> (then, Profile:PDA Dragon Cart Narrative Editor Preview Profile Data Object Reference, IsReset:bool=~) {is=title K2Node_FunctionEntry}]}
]
variables [
 IsReset {type=boolean}
 OnProfileSelected {}
 Profile {type=PDA_DragonCartNarrative_EditorPreviewProfileData_C, default=null}
]
widgets {
 info {inheritedWidgetCount=0, parentClass=&UserWidget, rootWidgetClass=&Button, widgetCount=4}
 widgets [
  Button_0 {
   widgetClassPath=&Button
   widget=Button_0
   properties {
    backgroundColor=<rgba 0.8802 0.8802 0.8802 1>
    widgetStyle {
     clickedSlateSound {resourceObject=null}
     disabled {
      bIsDynamicallyLoaded=false
      drawAs=NoDrawType
      imageType=NoImage
      mirroring=NoMirror
      resourceName=None
      resourceObject=null
      tiling=NoTile
      imageSize=<0 0>
      margin=<ltrb 0 0 0 0>
      outlineSettings {
       bUseBrushTransparency=false
       roundingType=HalfHeightRadius
       width=0
       color {colorUseRule=UseColor_Specified, specifiedColor=<rgba 0 0 0 0>}
       cornerRadii=<0 0 0 0>
      }
      tintColor {colorUseRule=UseColor_Specified, specifiedColor=<rgba 0 0 0 1>}
      uVRegion {bIsValid=false, max=<0 0>, min=<0 0>}
     }
     disabledForeground {colorUseRule=UseColor_Specified, specifiedColor=<rgba 0.5271 0.5271 0.5271 1>}
     hovered {
      bIsDynamicallyLoaded=false
      drawAs=RoundedBox
      imageType=NoImage
      mirroring=NoMirror
      resourceName=None
      resourceObject=null
      tiling=NoTile
      imageSize=<32 32>
      margin=<ltrb 0 0 0 0>
      outlineSettings {
       bUseBrushTransparency=true
       roundingType=FixedRadius
       width=1
       color {colorUseRule=UseColor_Specified, specifiedColor=<rgba 0.7243 0.7243 0.7243 1>}
       cornerRadii=<4 4 4 4>
      }
      tintColor {colorUseRule=UseColor_Specified, specifiedColor=<rgba 0.0729 0.0729 0.0729 1>}
      uVRegion {bIsValid=false, max=<0 0>, min=<0 0>}
     }
     hoveredForeground {colorUseRule=UseColor_Specified, specifiedColor=<rgba 0 1 1 1>}
     hoveredSlateSound {resourceObject=null}
     normal {
      bIsDynamicallyLoaded=false
      drawAs=RoundedBox
      imageType=NoImage
      mirroring=NoMirror
      resourceName=None
      resourceObject=null
      tiling=NoTile
      imageSize=<32 32>
      margin=<ltrb 0 0 0 0>
      outlineSettings {
       bUseBrushTransparency=true
       roundingType=FixedRadius
       width=2
       color {colorUseRule=UseColor_Specified, specifiedColor=<rgba 0.0521 0.0521 0.0521 1>}
       cornerRadii=<5 5 5 5>
      }
      tintColor {colorUseRule=UseColor_Specified, specifiedColor=<rgba 0.0312 0.0312 0.0312 1>}
      uVRegion {bIsValid=false, max=<0 0>, min=<0 0>}
     }
     normalForeground {colorUseRule=UseColor_Specified, specifiedColor=<rgba 0 1 1 1>}
     normalPadding=<ltrb 12 1.5 12 1.5>
     pressed {
      bIsDynamicallyLoaded=false
      drawAs=RoundedBox
      imageType=NoImage
      mirroring=NoMirror
      resourceName=None
      resourceObject=null
      tiling=NoTile
      imageSize=<32 32>
      margin=<ltrb 0 0 0 0>
      outlineSettings {
       bUseBrushTransparency=true
       roundingType=FixedRadius
       width=1
       color {colorUseRule=UseColor_Specified, specifiedColor=<rgba 0.7243 0.7243 0.7243 1>}
       cornerRadii=<4 4 4 4>
      }
      tintColor {colorUseRule=UseColor_Specified, specifiedColor=<rgba 0.0625 0.0625 0.0625 1>}
      uVRegion {bIsValid=false, max=<0 0>, min=<0 0>}
     }
     pressedForeground {colorUseRule=UseColor_Specified, specifiedColor=<rgba 0 1 1 1>}
     pressedPadding=<ltrb 12 1.5 12 1.5>
     pressedSlateSound {resourceObject=null}
    }
   }
  }
  VerticalBox_98 {
   widgetClassPath=&VerticalBox
   parent=Button_0
   widget=VerticalBox_98
   properties {bOverride_Cursor=true}
   slot {class=ButtonSlot, properties {horizontalAlignment=HAlign_Fill, verticalAlignment=VAlign_Fill, padding=<ltrb 0 0 0 0>}}
  }
  ScaleBox_119 {
   widgetClassPath=&ScaleBox
   parent=VerticalBox_98
   widget=ScaleBox_119
   slot {
    class=VerticalBoxSlot
    properties {horizontalAlignment=HAlign_Center, verticalAlignment=VAlign_Center, padding=<ltrb 0 0 0 10.05>, size {sizeRule=Fill, value=0.1}}
   }
  }
  Label {
   widgetClassPath=&TextBlock
   bIsVariable=true
   parent=ScaleBox_119
   widget=Label
   properties {
    bOverride_Cursor=true
    justification=Center
    text=DragonCartNarrative - Editor Preview Tool
    visibility=SelfHitTestInvisible
    shadowOffset=<0 0>
   }
   slot {class=ScaleBoxSlot, properties {horizontalAlignment=HAlign_Fill, verticalAlignment=VAlign_Fill}}
  }
 ]
}
