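# Editor utility widget for previewing DragonCartNarrative stages from a list of profile data assets, such as rift openings and shard deliveries.
# On construct it fills a scrollable grid with one option widget per profile and binds each selection event.
# Selecting unloads the current profile, loads the new one through a preview manager object and sets the rift material collection's RiftScale.
# Reset selections and closing the widget unload the active profile.
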EditorUtilityWidgetBlueprint /DragonCartNarrative/Gameplay/EditorUtilities/EditorPreview/WBP_DragonCartNarrative_EditorPreviewer
graphs [
 EventGraph {nodes [
   Event PreConstruct()
   > n1 = Construct Object from Class(Class=&/BP_DragonCartNarrative_PreviewManager.BP_DragonCartNarrative_PreviewManager_C) -> (ReturnValue:BP Dragon Cart Narrative Preview Manager Object Reference)
   > Set EditorPreviewManager(EditorPreviewManager:BP Dragon Cart Narrative Preview Manager Object Reference=@n1) -> (Output_Get:BP Dragon Cart Narrative Preview Manager Object Reference)
   Event Construct()
   > PopulateGrid()
   Event Destruct()
   > Is Valid(InputObject=Get CurrentSelectedProfile() -> (CurrentSelectedProfile:PDA Dragon Cart Narrative Editor Preview Profile Data Object Reference))
   > UnloadProfile!(
    execute
    self:BP Dragon Cart Narrative Preview Manager Object Reference=Get EditorPreviewManager() -> (EditorPreviewManager:BP Dragon Cart Narrative Preview Manager Object Reference)
    EditorPreviewProfile:PDA Dragon Cart Narrative Editor Preview Profile Data Object Reference=Get CurrentSelectedProfile()
   ) -> (then)
   n2 = Get EditorPreviewManager()
   > Is Valid(InputObject=@n2)
   > ObjectDestroyed!(execute, Object:Object Reference=@n2, Author:String) -> (then)
  ]}
 GetProfileByName {nodes [
   n1 = GetProfileByName!() -> (then, Name:String) {is=title K2Node_FunctionEntry}
   > n2 = For Each Loop(Array:Array of PDA Dragon Cart Narrative Editor Preview Profile Data Object References=Get PreviewProfiles() -> (PreviewProfiles:Array of PDA Dragon Cart Narrative Editor Preview Profile Data Object References)) -> (Array Element:PDA Dragon Cart Narrative Editor Preview Profile Data Object Reference)
   Return Node!(execute=@n2.Completed, Success:bool=false, Profile:PDA Dragon Cart Narrative Editor Preview Profile Data Object Reference)
   n3 = Get Name!(self:PDA Dragon Cart Narrative Editor Preview Profile Data Object Reference=@n2) -> (Name:Name=None)
   > Branch(Condition="Equal, Case Insensitive (String)"(A="To String (Name)"(InName=@n3), B=@n1))
   > Return Node(Success=true, Profile=@n2)
  ]}
 OnProfileSelected {nodes [
   n1 = OnProfileSelected!() -> (then, Profile:PDA Dragon Cart Narrative Editor Preview Profile Data Object Reference, IsReset:bool=~) {is=title K2Node_FunctionEntry}
   > n2 = Branch(Condition=@n1.IsReset)
   > Is Valid(InputObject=Get CurrentSelectedProfile())
   > UnloadProfile(self=Get EditorPreviewManager(), EditorPreviewProfile=Get CurrentSelectedProfile())
   > Set CurrentSelectedProfile(CurrentSelectedProfile:PDA Dragon Cart Narrative Editor Preview Profile Data Object Reference) -> (Output_Get:PDA Dragon Cart Narrative Editor Preview Profile Data Object Reference)
   n3 = Is Valid(exec=@n2.else, InputObject=Get CurrentSelectedProfile())
   > UnloadProfile(self=Get EditorPreviewManager(), EditorPreviewProfile=Get CurrentSelectedProfile())
   > n4 = Set CurrentSelectedProfile(CurrentSelectedProfile=@n1)
   LoadProfile!(
    execute=[@n4.then, @n3.Is Not Valid]
    self:BP Dragon Cart Narrative Preview Manager Object Reference=Get EditorPreviewManager()
    EditorPreviewProfile:PDA Dragon Cart Narrative Editor Preview Profile Data Object Reference=@n1
   ) -> (then)
   > Set CurrentSelectedProfile(CurrentSelectedProfile=@n1)
   > SetScalarParameterValue!(
    execute
    WorldContextObject=Self-Reference()
    Collection:Material Parameter Collection Object Reference=/DragonCartNarrative/Assets/VFX/RiftPortalIdle/MPC_DragonCart_Orb_Rift
    ParameterName:Name=RiftScale
    ParameterValue:float=Get!(self:PDA Dragon Cart Narrative Editor Preview Profile Data Object Reference=@n1) -> (
     SillyTunaProfileData_SetEditorCamera_1_#1:bool=~
     SillyTunaProfileData_Location_6_#2:Vector=~
     SillyTunaProfileData_Rotation_7_#3:Rotator=~
     SillyTunaProfileData_RiftScale_10_#4:double=~
    ).SillyTunaProfileData_RiftScale_10_#4
   ) -> (then) {is=title K2Node_CallMaterialParameterCollectionFunction}
  ]}
 PopulateGrid {nodes [
   PopulateGrid() {is=title K2Node_FunctionEntry}
   n1 = Get PreviewProfiles()
   > n2 = For Loop(LastIndex=int - int(A=Length(TargetArray:Array of PDA Dragon Cart Narrative Editor Preview Profile Data Object References=@n1)))
   > SpawnAndInitProfileElement!(
    execute
    self
    Profile:PDA Dragon Cart Narrative Editor Preview Profile Data Object Reference="Get (a copy)"(Array:Array of PDA Dragon Cart Narrative Editor Preview Profile Data Object References=Get PreviewProfiles(), Dimension 1=@n2) -> (Output:PDA Dragon Cart Narrative Editor Preview Profile Data Object Reference)
    IsReset:bool=false
    Row:int=@n2
    Column:int=0
   ) -> (then)
  ]}
 SpawnAndInitProfileElement {nodes [
   n1 = SpawnAndInitProfileElement!() -> (then, Profile:PDA Dragon Cart Narrative Editor Preview Profile Data Object Reference, IsReset:bool=~, Row:int, Column:int) {is=title K2Node_FunctionEntry}
   > n2 = Create Widget!(
    execute:Exec
    WorldContextObject:Object Reference=Self-Reference()
    Class:User Widget Class Reference=&/WBP_DragonCartNarrative_EditorPreviewOption.WBP_DragonCartNarrative_EditorPreviewOption_C
    OwningPlayer:Player Controller Object Reference
   ) -> (then:Exec, ReturnValue:WBP Dragon Cart Narrative Editor Preview Option Object Reference)
   > AddChildToGrid!(
    execute
    self:Grid Panel Object Reference=Get GridContainer() -> (GridContainer:Grid Panel Object Reference)
    Content:Widget Object Reference=@n2
    InRow:int=Get Row() -> (Row:int=0)
    InColumn:int=Get Column() -> (Column:int=0)
   ) -> (then, ReturnValue:Grid Slot Object Reference)
   > Init!(
    execute
    self:WBP Dragon Cart Narrative Editor Preview Option Object Reference=[@n2, @n2]
    Profile:PDA Dragon Cart Narrative Editor Preview Profile Data Object Reference=@n1
    IsReset:bool=@n1.IsReset
   ) -> (then)
   > Bind Event to On Profile Selected(self:WBP Dragon Cart Narrative Editor Preview Option Object Reference=@n2, Delegate=Create Event()) {is=title K2Node_AddDelegate}
  ]}
]
variables [
 CurrentSelectedProfile {type=PDA_DragonCartNarrative_EditorPreviewProfileData_C, default=null}
 EditorPreviewManager {type=BP_DragonCartNarrative_PreviewManager_C, default=null}
 PreviewProfiles {
  type="array<PDA_DragonCartNarrative_EditorPreviewProfileData_C>"
  default [
   /DragonCartNarrative/Gameplay/EditorUtilities/EditorPreview/Profiles/DA_ShardShooting
   /DragonCartNarrative/Gameplay/EditorUtilities/EditorPreview/Profiles/DA_RiftOpening_1
   /DragonCartNarrative/Gameplay/EditorUtilities/EditorPreview/Profiles/DA_RiftOpening_2
   /DragonCartNarrative/Gameplay/EditorUtilities/EditorPreview/Profiles/DA_OrbRiftOut
   /DragonCartNarrative/Gameplay/EditorUtilities/EditorPreview/Profiles/DA_NoOrb
   /DragonCartNarrative/Gameplay/EditorUtilities/EditorPreview/Profiles/DA_RiftOpeningNoOrb1
   /DragonCartNarrative/Gameplay/EditorUtilities/EditorPreview/Profiles/DA_RiftOpeningNoOrb2
   /DragonCartNarrative/Gameplay/EditorUtilities/EditorPreview/Profiles/DA_OrbRiftIn2
   /DragonCartNarrative/Gameplay/EditorUtilities/EditorPreview/Profiles/DA_ShardDeliveryHQ_1
   /DragonCartNarrative/Gameplay/EditorUtilities/EditorPreview/Profiles/DA_ShardDeliveryHQ_2
   /DragonCartNarrative/Gameplay/EditorUtilities/EditorPreview/Profiles/DA_ShardDeliveryHQ3
   /DragonCartNarrative/Gameplay/EditorUtilities/EditorPreview/Profiles/DA_ShardDeliveryHQ_4
   /DragonCartNarrative/Gameplay/EditorUtilities/EditorPreview/Profiles/DA_ShardDeliveryHQ_4_Post
   /DragonCartNarrative/Gameplay/EditorUtilities/EditorPreview/Profiles/DA_Shock
   /DragonCartNarrative/Gameplay/EditorUtilities/EditorPreview/Profiles/DA_RealityRifts
  ]
 }
]
widgets {
 info {rootWidgetClass=&VerticalBox, widgetCount=5}
 widgets [
  VerticalBox_0 {widgetClassPath=&VerticalBox, widget=VerticalBox_0, properties {bOverride_Cursor=true}}
  ScaleBox_0 {
   widgetClassPath=&ScaleBox
   parent=VerticalBox_0
   widget=ScaleBox_0
   slot {class=VerticalBoxSlot, properties {padding=<ltrb 150 10 150 10>}}
  }
  TextBlock_175 {
   widgetClassPath=&TextBlock
   parent=ScaleBox_0
   widget=TextBlock_175
   properties {
    bOverride_Cursor=true
    justification=Center
    text=DragonCartNarrative - Editor Preview Tool
    visibility=SelfHitTestInvisible
    shadowOffset=<0 0>
   }
   slot {class=ScaleBoxSlot}
  }
  AthenaScrollBox_298 {
   widgetClassPath=/Script/FortniteUI.AthenaScrollBox
   parent=VerticalBox_0
   widget=AthenaScrollBox_298
   properties {
    widgetBarStyle {
     thickness=8
     draggedThumbImage {
      bIsDynamicallyLoaded=false
      drawAs=RoundedBox
      imageType=NoImage
      mirroring=NoMirror
      resourceName=None
      resourceObject=null
      tiling=NoTile
      imageSize=<0 0>
      margin=<ltrb 0 0 0 0>
      outlineSettings {
       bUseBrushTransparency=false
       roundingType=FixedRadius
       width=0
       color {colorUseRule=UseColor_Specified, specifiedColor=<rgba 0 0 0 0>}
       cornerRadii=<4 4 4 4>
      }
      tintColor {colorUseRule=UseColor_Specified, specifiedColor=<rgba 0.2159 0.2159 0.2159 1>}
      uVRegion {bIsValid=false, max=<0 0>, min=<0 0>}
     }
     horizontalBackgroundImage {
      bIsDynamicallyLoaded=false
      drawAs=NoDrawType
      imageType=NoImage
      mirroring=NoMirror
      resourceName=None
      resourceObject=null
      tiling=NoTile
      imageSize=<0 0>
      margin=<ltrb 0 0 0 0>
      outlineSettings {
       bUseBrushTransparency=false
       roundingType=HalfHeightRadius
       width=0
       color {colorUseRule=UseColor_Specified, specifiedColor=<rgba 0 0 0 0>}
       cornerRadii=<0 0 0 0>
      }
      tintColor {colorUseRule=UseColor_Specified, specifiedColor=<rgba 1 1 1 1>}
      uVRegion {bIsValid=false, max=<0 0>, min=<0 0>}
     }
     horizontalBottomSlotImage {
      bIsDynamicallyLoaded=false
      drawAs=NoDrawType
      imageType=NoImage
      mirroring=NoMirror
      resourceName=None
      resourceObject=null
      tiling=NoTile
      imageSize=<0 0>
      margin=<ltrb 0 0 0 0>
      outlineSettings {
       bUseBrushTransparency=false
       roundingType=HalfHeightRadius
       width=0
       color {colorUseRule=UseColor_Specified, specifiedColor=<rgba 0 0 0 0>}
       cornerRadii=<0 0 0 0>
      }
      tintColor {colorUseRule=UseColor_Specified, specifiedColor=<rgba 1 1 1 1>}
      uVRegion {bIsValid=false, max=<0 0>, min=<0 0>}
     }
     horizontalTopSlotImage {
      bIsDynamicallyLoaded=false
      drawAs=NoDrawType
      imageType=NoImage
      mirroring=NoMirror
      resourceName=None
      resourceObject=null
      tiling=NoTile
      imageSize=<0 0>
      margin=<ltrb 0 0 0 0>
      outlineSettings {
       bUseBrushTransparency=false
       roundingType=HalfHeightRadius
       width=0
       color {colorUseRule=UseColor_Specified, specifiedColor=<rgba 0 0 0 0>}
       cornerRadii=<0 0 0 0>
      }
      tintColor {colorUseRule=UseColor_Specified, specifiedColor=<rgba 1 1 1 1>}
      uVRegion {bIsValid=false, max=<0 0>, min=<0 0>}
     }
     hoveredThumbImage {
      bIsDynamicallyLoaded=false
      drawAs=RoundedBox
      imageType=NoImage
      mirroring=NoMirror
      resourceName=None
      resourceObject=null
      tiling=NoTile
      imageSize=<0 0>
      margin=<ltrb 0 0 0 0>
      outlineSettings {
       bUseBrushTransparency=false
       roundingType=FixedRadius
       width=0
       color {colorUseRule=UseColor_Specified, specifiedColor=<rgba 0 0 0 0>}
       cornerRadii=<4 4 4 4>
      }
      tintColor {colorUseRule=UseColor_Specified, specifiedColor=<rgba 0.2159 0.2159 0.2159 1>}
      uVRegion {bIsValid=false, max=<0 0>, min=<0 0>}
     }
     normalThumbImage {
      bIsDynamicallyLoaded=false
      drawAs=RoundedBox
      imageType=NoImage
      mirroring=NoMirror
      resourceName=None
      resourceObject=null
      tiling=NoTile
      imageSize=<0 0>
      margin=<ltrb 0 0 0 0>
      outlineSettings {
       bUseBrushTransparency=false
       roundingType=FixedRadius
       width=0
       color {colorUseRule=UseColor_Specified, specifiedColor=<rgba 0 0 0 0>}
       cornerRadii=<4 4 4 4>
      }
      tintColor {colorUseRule=UseColor_Specified, specifiedColor=<rgba 0.0953 0.0953 0.0953 1>}
      uVRegion {bIsValid=false, max=<0 0>, min=<0 0>}
     }
     verticalBackgroundImage {
      bIsDynamicallyLoaded=false
      drawAs=NoDrawType
      imageType=NoImage
      mirroring=NoMirror
      resourceName=None
      resourceObject=null
      tiling=NoTile
      imageSize=<0 0>
      margin=<ltrb 0 0 0 0>
      outlineSettings {
       bUseBrushTransparency=false
       roundingType=HalfHeightRadius
       width=0
       color {colorUseRule=UseColor_Specified, specifiedColor=<rgba 0 0 0 0>}
       cornerRadii=<0 0 0 0>
      }
      tintColor {colorUseRule=UseColor_Specified, specifiedColor=<rgba 1 1 1 1>}
      uVRegion {bIsValid=false, max=<0 0>, min=<0 0>}
     }
     verticalBottomSlotImage {
      bIsDynamicallyLoaded=false
      drawAs=NoDrawType
      imageType=NoImage
      mirroring=NoMirror
      resourceName=None
      resourceObject=null
      tiling=NoTile
      imageSize=<0 0>
      margin=<ltrb 0 0 0 0>
      outlineSettings {
       bUseBrushTransparency=false
       roundingType=HalfHeightRadius
       width=0
       color {colorUseRule=UseColor_Specified, specifiedColor=<rgba 0 0 0 0>}
       cornerRadii=<0 0 0 0>
      }
      tintColor {colorUseRule=UseColor_Specified, specifiedColor=<rgba 1 1 1 1>}
      uVRegion {bIsValid=false, max=<0 0>, min=<0 0>}
     }
     verticalTopSlotImage {
      bIsDynamicallyLoaded=false
      drawAs=NoDrawType
      imageType=NoImage
      mirroring=NoMirror
      resourceName=None
      resourceObject=null
      tiling=NoTile
      imageSize=<0 0>
      margin=<ltrb 0 0 0 0>
      outlineSettings {
       bUseBrushTransparency=false
       roundingType=HalfHeightRadius
       width=0
       color {colorUseRule=UseColor_Specified, specifiedColor=<rgba 0 0 0 0>}
       cornerRadii=<0 0 0 0>
      }
      tintColor {colorUseRule=UseColor_Specified, specifiedColor=<rgba 1 1 1 1>}
      uVRegion {bIsValid=false, max=<0 0>, min=<0 0>}
     }
    }
    widgetStyle {
     barThickness=8
     bottomShadowBrush {
      bIsDynamicallyLoaded=false
      drawAs=Box
      imageType=FullColor
      mirroring=NoMirror
      resourceName=../../../Engine/Content/Slate/Common/ScrollBoxShadowBottom.png
      resourceObject=null
      tiling=NoTile
      imageSize=<16 8>
      margin=<ltrb 0.5 0 0.5 1>
      outlineSettings {
       bUseBrushTransparency=false
       roundingType=HalfHeightRadius
       width=0
       color {colorUseRule=UseColor_Specified, specifiedColor=<rgba 0 0 0 0>}
       cornerRadii=<0 0 0 0>
      }
      tintColor {colorUseRule=UseColor_Specified, specifiedColor=<rgba 1 1 1 1>}
      uVRegion {bIsValid=false, max=<0 0>, min=<0 0>}
     }
     horizontalScrolledContentPadding=<ltrb 0 0 1 0>
     leftShadowBrush {
      bIsDynamicallyLoaded=false
      drawAs=Box
      imageType=FullColor
      mirroring=NoMirror
      resourceName=../../../Engine/Content/Slate/Common/ScrollBoxShadowLeft.png
      resourceObject=null
      tiling=NoTile
      imageSize=<8 16>
      margin=<ltrb 1 0.5 0 0.5>
      outlineSettings {
       bUseBrushTransparency=false
       roundingType=HalfHeightRadius
       width=0
       color {colorUseRule=UseColor_Specified, specifiedColor=<rgba 0 0 0 0>}
       cornerRadii=<0 0 0 0>
      }
      tintColor {colorUseRule=UseColor_Specified, specifiedColor=<rgba 1 1 1 1>}
      uVRegion {bIsValid=false, max=<0 0>, min=<0 0>}
     }
     rightShadowBrush {
      bIsDynamicallyLoaded=false
      drawAs=Box
      imageType=FullColor
      mirroring=NoMirror
      resourceName=../../../Engine/Content/Slate/Common/ScrollBoxShadowRight.png
      resourceObject=null
      tiling=NoTile
      imageSize=<8 16>
      margin=<ltrb 0 0.5 1 0.5>
      outlineSettings {
       bUseBrushTransparency=false
       roundingType=HalfHeightRadius
       width=0
       color {colorUseRule=UseColor_Specified, specifiedColor=<rgba 0 0 0 0>}
       cornerRadii=<0 0 0 0>
      }
      tintColor {colorUseRule=UseColor_Specified, specifiedColor=<rgba 1 1 1 1>}
      uVRegion {bIsValid=false, max=<0 0>, min=<0 0>}
     }
     topShadowBrush {
      bIsDynamicallyLoaded=false
      drawAs=Box
      imageType=FullColor
      mirroring=NoMirror
      resourceName=../../../Engine/Content/Slate/Common/ScrollBoxShadowTop.png
      resourceObject=null
      tiling=NoTile
      imageSize=<16 8>
      margin=<ltrb 0.5 1 0.5 0>
      outlineSettings {
       bUseBrushTransparency=false
       roundingType=HalfHeightRadius
       width=0
       color {colorUseRule=UseColor_Specified, specifiedColor=<rgba 0 0 0 0>}
       cornerRadii=<0 0 0 0>
      }
      tintColor {colorUseRule=UseColor_Specified, specifiedColor=<rgba 1 1 1 1>}
      uVRegion {bIsValid=false, max=<0 0>, min=<0 0>}
     }
     verticalScrolledContentPadding=<ltrb 0 0 0 1>
    }
   }
   slot {class=VerticalBoxSlot, properties {size {sizeRule=Fill, value=1}}}
  }
  GridContainer {
   widgetClassPath=&GridPanel
   bIsVariable=true
   parent=AthenaScrollBox_298
   widget=GridContainer
   slot {class=ScrollBoxSlot, properties {horizontalAlignment=HAlign_Center, padding=<ltrb 5 0 5 5>, size {sizeRule=Fill, value=1}}}
  }
 ]
}
